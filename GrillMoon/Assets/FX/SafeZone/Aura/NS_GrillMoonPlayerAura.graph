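# Niagara system for the GrillMoon safe-zone player aura, drawing teal flames and glow on a skeletal mesh.
# Flames_PerBone pins a small burst of mesh and sprite particles to bones, storing time-smoothed positions for trailing interpolation and switching alpha for the front end.
# Glow adds one fade-in mesh particle, only for the local player or front-end preview.
# Arms/legs, GPU per-triangle and Minimal emitters are disabled.
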NiagaraSystem /GrillMoon/Assets/FX/SafeZone/Aura/NS_GrillMoonPlayerAura
dependencies {
 usedDataInterfaces [
  &NiagaraDataInterfaceCurve
  &NiagaraDataInterfaceArrayFloat
  &NiagaraDataInterfaceArrayInt32
  &NiagaraDataInterfaceCamera
 ]
 usedDynamicInputs [
  &FN_Select_Float_from_Time_Of_Day
  /Game/Weapons/FORT_Melee/Pickaxe_GateHound/FX/VariantManager/VariantVectorSelector
  &FN_Select_Vector_from_Time_Of_Day
 ]
 usedModules [
  &self:SystemSpawnScript.SystemScriptSource.SystemScriptGraph.NiagaraNodeAssignment_0.SetVariables_#1
  &self:SystemSpawnScript.SystemScriptSource.SystemScriptGraph.NiagaraNodeAssignment_2.SetVariables_#2
  &SystemState
  &Timeline
  /Game/Weapons/FORT_Melee/Pickaxe_GateHound/FX/VariantManager/FN_VariantManager
  &self:SystemSpawnScript.SystemScriptSource.SystemScriptGraph.NiagaraNodeAssignment_7.SetVariables_#3
  &self:SystemSpawnScript.SystemScriptSource.SystemScriptGraph.NiagaraNodeAssignment_8.SetVariables_#4
 ]
}
emitters {
 FlamesArms_legs {
  bEnabled=false
  emitterUpdateScript [
   &EmitterState(
    Inactive Response=::NewEnumerator1
    Life Cycle Mode=::NewEnumerator1
    Loop Behavior=::NewEnumerator1
    Loop Duration Mode=::NewEnumerator1
    UseLoopDelay=false
   )
   &SpawnBurst_Instantaneous(Spawn Count=4, Use Spawn Probability=false, UseLoopCountLimit=false)
  ]
  particleSpawnScript [
   &PartitionParticles(Number Of Partitions=25)
   &InitializeParticle(
    Color=<rgba 0.0018 0.0885 0.0857 1>
    Lifetime Max=12
    Lifetime Min=12
    Lifetime Mode=::NewEnumerator1
    Mesh Scale Mode=::NewEnumerator3
    Mesh Uniform Scale=&SelectFloatFromArray(
     Direct Array Index=$Particles.PartitionParticles.Partition
     Float Selection Array=DataInterface() {propertyValues {FloatData [1.25,1.25,1,1]}}
    )
    Sprite Size Mode=::NewEnumerator3
    Uniform Sprite Size=&SelectFloatFromArray(
     Direct Array Index=$Particles.PartitionParticles.Partition
     Float Selection Array=DataInterface() {propertyValues {FloatData [4,4,4,4,4,4,4,4,4,4,4,4,4,4,4,4,4,4,0,0,0,0,0,0,0,0,0]}}
    )
    UsePositionOffset=false
    Write Material Random=true
   )
   &InitialMeshOrientation(
    Mesh Axis To Orient=<0 0 -1>
    Mesh Orientation Mode=::NewEnumerator2
    Orientation Coordinate Space=::World
    Orientation Vector=<0 0 1>
    Rotation=<0.25 0 0>
    Use Rotation Vector=false
   )
   &DynamicMaterialParameters(
    Index 0 Param 1=&RandomRangeFloat(Minimum=0, Maximum=1)
    Index 0 Param 2=2
    Module.Index0Param0Write=true
    Module.Index0Param1Write=true
    Module.Index0Param2Write=true
    Module.Index0Param3Write=true
   )
   &CameraOffset(Camera Offset Amount=35)
  ]
  particleUpdateScript [
   &ParticleState(Kill Particles When Lifetime Has Elapsed=false, Let Infinitely Lived Particles Die When Emitter Deactivates=true)
   /SetVariables(Particles.AlphaScale:NiagaraFloat=&FN_FrontEndFloatModifier(
     Gameplay=&Lerp_Float(A=1, B=1, Alpha=$System.Timeline.NormalizedTime)
     Front End=&MakeCustomFloatFromBool(Bool=$User.bIsFrontEnd, True Float=$System.FrontEndGlowLoop, False Float=0)
    )) {isSetParameters=true, module=SetVariables_#5}
   &ScaleMeshSize(Scale Factor=<1 1 2>)
   &SolveForcesAndVelocity(Clamp Velocity=false, Limit Acceleration=false)
   &/GrillMoon_SkeletalMeshLocation_withParentboneout(MaskSpawnGroup=false, OverrideRandomSeed=false, Skeletal Mesh:NiagaraDataInterfaceSkeletalMesh=$User.SK)
   &SkeletalMeshLocation(
    Bone / Socket Index=&Add_Integer(A=&ReturnExecIndex(), B=1)
    Bone Sampling Mode=::NewEnumerator6
    MaskSpawnGroup=false
    Orientation Sampling=::NewEnumerator1
    Skeletal Mesh=$User.SK
   )
   /SetVariables(Particles.Position=&Lerp_Vector(A=<0 0 0>, B=<1 1 1>, Alpha=&RandomRangeFloat(Minimum=0, Maximum=1))) {isSetParameters=true}
   &ScaleColor(Scale Alpha=$Particles.AlphaScale, Scale RGB=&VectorFromFloat(Value=3.5), ScaleA=true, ScaleRGB=true)
   /SetVariables(Particles.Color_02:LinearColor=<rgba 0.0018 0.0885 0.0857 1>) {isSetParameters=true, module=SetVariables_#6}
   /SetVariables(Particles.Color_03:LinearColor=<rgba 0.2203 0.9 0.7754 1>) {isSetParameters=true, module=SetVariables_#7}
   &InterpolateOverTime(Rate Of Change=3, Target Position Value:NiagaraPosition=$Particles.Position, Variable Type=::NewEnumerator6)
   /SetVariables(Particles.Interpolate Dynamic=&MakeFloat4FromVectorAndFloat(
     "Vector (RGB)"=$Particles.InterpolateOverTime.InterpolateOverTime_Position.Moving Average
     "Float (Alpha)"=&Add_Float(A=$User.Interpolate Intensity, B=-0.5)
    )) {isSetParameters=true, module=SetVariables_#8}
   &FN_Sprite_Attributes(
    Camera Offset:NiagaraFloat=&SelectFloatFromArray(
     Direct Array Index=$Particles.PartitionParticles.Partition
     Float Selection Array=DataInterface() {propertyValues {FloatData [2,2,5,5,5,2,2,5,0,5,2,1.5,1.5,1,1,1.5,1,1.5]}}
    )
    Color=<rgba 0.0473 0.4323 0.3617 1>
    Create Camera Offset Attribute=true
    Create Color Attribute=true
    Write to Module Namespace=true
   )
   /SetVariables(Particles.AirglowDyn:Vector4f=<2 1 0 0>) {isSetParameters=true, module=SetVariables_#9}
   /SetVariables(
    Particles.DecalDyn:Vector4f=&MakeVector4(
     X=0.8
     Y=0.5
     Z=&RandomRangeFloat(Minimum=0, Maximum=1)
     W=&RandomRangeFloat(Minimum=0, Maximum=0)
    )
    Particles.FireDecalColor:LinearColor=<rgba 0.1721 0.6323 0.7031 1>
    Particles.FireDecalSize:Vector2f=&Multiply_Vector2D_ByFloat(
     Float=&Multiply_Float(
      A=&SelectFloatFromArray(
       Direct Array Index=$Particles.PartitionParticles.Partition
       Float Selection Array=DataInterface() {propertyValues {FloatData [0.5,0.5,0.3,0.5,0.8,0.5,0.5,0.3,0.5,0.8,1,0.5,0.3,0.5,0.3,0.5,0.3,0.5,0,0,0,0,0,0,0]}}
      )
      B=1.2
     )
     Vector2D=$Particles.SpriteSize
    )
    Particles.GlowDecalColor:LinearColor=<rgba 0.0054 0.1413 0.195 1>
   ) {isSetParameters=true, module=SetVariables_#10}
   /SetVariables(Particles.VisibilityTag=&MakeCustomIntFromBool(
     Bool=$User.bisFootHidden
     True INT=&self:ScratchDynamicInput(
      Masked Int:NiagaraInt32=1
      Partition Mask:NiagaraInt32=&SelectIntFromArray(
       Direct Array Index=&ReturnExecIndex()
       Int Selection Array=DataInterface() {propertyValues {IntData [-1,-1,-1,-1,-1,-1,-1,-1,-1,-1,-1,-1,-1,-1,-1,-1,-1,-1,18,19,20,21,22,23,24,25]}}
      )
      Passthrough Int:NiagaraInt32=0
     )
     False INT=0
    )) {isSetParameters=true, module=SetVariables_#11}
  ]
  rendererClasses [&NiagaraMeshRendererProperties, &NiagaraSpriteRendererProperties]
  renderers [
   {rendererClass=&NiagaraMeshRendererProperties, rendererIndex=0}
   {rendererClass=&NiagaraMeshRendererProperties, rendererIndex=1}
   {rendererClass=&NiagaraMeshRendererProperties, rendererIndex=2}
   {rendererClass=&NiagaraMeshRendererProperties, rendererIndex=3}
   {rendererClass=&NiagaraMeshRendererProperties, rendererIndex=4}
   {rendererClass=&NiagaraMeshRendererProperties, rendererIndex=5}
   {rendererClass=&NiagaraSpriteRendererProperties, rendererIndex=6}
   {rendererClass=&NiagaraSpriteRendererProperties, rendererIndex=7}
   {rendererClass=&NiagaraSpriteRendererProperties, rendererIndex=8}
   {rendererClass=&NiagaraSpriteRendererProperties, rendererIndex=9}
   {rendererClass=&NiagaraSpriteRendererProperties, rendererIndex=10}
   {rendererClass=&NiagaraSpriteRendererProperties, rendererIndex=11}
  ]
 }
 Flames_PerBone {
  emitterUpdateScript [
   &EmitterState(
    Inactive Response=::NewEnumerator1
    Life Cycle Mode=::NewEnumerator1
    Loop Behavior=::NewEnumerator1
    Loop Duration Mode=::NewEnumerator1
    UseLoopDelay=false
   )
   &SpawnBurst_Instantaneous(Spawn Count=4, Use Spawn Probability=false, UseLoopCountLimit=false)
  ]
  particleSpawnScript [
   &PartitionParticles(Number Of Partitions=25)
   &InitializeParticle(
    Color=<rgba 0.0018 0.0885 0.0857 1>
    Lifetime Max=12
    Lifetime Min=12
    Lifetime Mode=::NewEnumerator1
    Mesh Scale Mode=::NewEnumerator3
    Mesh Uniform Scale=&SelectFloatFromArray(
     Direct Array Index=$Particles.PartitionParticles.Partition
     Float Selection Array=DataInterface() {propertyValues {FloatData [1.25,1.25,1,1]}}
    )
    Sprite Size Mode=::NewEnumerator3
    Uniform Sprite Size=&SelectFloatFromArray(
     Direct Array Index=$Particles.PartitionParticles.Partition
     Float Selection Array=DataInterface() {propertyValues {FloatData [4,4,4,4,4,4,4,4,4,4,4,4,4,4,4,4,4,4,0,0,0,0,0,0,0,0,0]}}
    )
    UsePositionOffset=false
    Write Material Random=true
   )
   &InitialMeshOrientation(
    Mesh Axis To Orient=<0 0 -1>
    Mesh Orientation Mode=::NewEnumerator2
    Orientation Coordinate Space=::World
    Orientation Vector=<0 0 1>
    Rotation=<0.25 0 0>
    Use Rotation Vector=false
   )
   &DynamicMaterialParameters(
    Index 0 Param 1=&RandomRangeFloat(Minimum=0, Maximum=1)
    Index 0 Param 2=2
    Module.Index0Param0Write=true
    Module.Index0Param1Write=true
    Module.Index0Param2Write=true
    Module.Index0Param3Write=true
   )
   &CameraOffset(Camera Offset Amount=20)
  ]
  particleUpdateScript [
   &ParticleState(Kill Particles When Lifetime Has Elapsed=false, Let Infinitely Lived Particles Die When Emitter Deactivates=true)
   /SetVariables(Particles.AlphaScale:NiagaraFloat=&FN_FrontEndFloatModifier(
     Gameplay=&Lerp_Float(A=1, B=1, Alpha=$System.Timeline.NormalizedTime)
     Front End=&MakeCustomFloatFromBool(Bool=$User.bIsFrontEnd, True Float=$System.FrontEndGlowLoop, False Float=0)
    )) {isSetParameters=true, module=SetVariables_#5}
   &ScaleMeshSize(Scale Factor=<1 1 1.35>)
   &SolveForcesAndVelocity(Clamp Velocity=false, Limit Acceleration=false)
   &SkeletalMeshLocation(
    Bone / Socket Index=&Add_Integer(A=&ReturnExecIndex(), B=1)
    Bone Sampling Mode=::NewEnumerator4
    MaskSpawnGroup=false
    Orientation Sampling=::NewEnumerator1
    Skeletal Mesh=$System.TargetSM
   )
   &ScaleColor(Scale Alpha=$Particles.AlphaScale, Scale RGB=&VectorFromFloat(Value=3.5), ScaleA=true, ScaleRGB=true)
   /SetVariables(Particles.Color_02:LinearColor=<rgba 0.0018 0.0885 0.0857 1>) {isSetParameters=true, module=SetVariables_#6}
   /SetVariables(Particles.Color_03:LinearColor=<rgba 0.2203 0.9 0.7754 1>) {enabled=false, isSetParameters=true, module=SetVariables_#7}
   &InterpolateOverTime(Rate Of Change=3, Target Position Value:NiagaraPosition=$Particles.Position, Variable Type=::NewEnumerator6)
   /SetVariables(Particles.Interpolate Dynamic=&MakeFloat4FromVectorAndFloat(
     "Vector (RGB)"=$Particles.InterpolateOverTime.InterpolateOverTime_Position.Moving Average
     "Float (Alpha)"=&Add_Float(A=$User.Interpolate Intensity, B=-0.5)
    )) {isSetParameters=true, module=SetVariables_#8}
   /SetVariables(Particles.AirglowDyn:Vector4f=<2 1 0 0>) {enabled=false, isSetParameters=true, module=SetVariables_#9}
   /SetVariables(
    Particles.DecalDyn:Vector4f=&MakeVector4(
     X=0.8
     Y=0.5
     Z=&RandomRangeFloat(Minimum=0, Maximum=1)
     W=&RandomRangeFloat(Minimum=0, Maximum=0)
    )
    Particles.FireDecalColor:LinearColor=<rgba 0.1721 0.6323 0.7031 1>
    Particles.FireDecalSize:Vector2f=&Multiply_Vector2D_ByFloat(
     Float=&Multiply_Float(
      A=&SelectFloatFromArray(
       Direct Array Index=$Particles.PartitionParticles.Partition
       Float Selection Array=DataInterface() {propertyValues {FloatData [0.5,0.5,0.3,0.5,0.8,0.5,0.5,0.3,0.5,0.8,1,0.5,0.3,0.5,0.3,0.5,0.3,0.5,0,0,0,0,0,0,0]}}
      )
      B=1.2
     )
     Vector2D=$Particles.SpriteSize
    )
    Particles.GlowDecalColor:LinearColor=<rgba 0.0054 0.1413 0.195 1>
   ) {enabled=false, isSetParameters=true, module=SetVariables_#10}
   /SetVariables(Particles.VisibilityTag=&MakeCustomIntFromBool(
     Bool=$User.bisFootHidden
     True INT=&self:ScratchDynamicInput(
      Masked Int:NiagaraInt32=1
      Partition Mask:NiagaraInt32=&SelectIntFromArray(
       Direct Array Index=&ReturnExecIndex()
       Int Selection Array=DataInterface() {propertyValues {IntData [-1,-1,-1,-1,-1,-1,-1,-1,-1,-1,-1,-1,-1,-1,-1,-1,-1,-1,18,19,20,21,22,23,24,25]}}
      )
      Passthrough Int:NiagaraInt32=0
     )
     False INT=0
    )) {isSetParameters=true, module=SetVariables_#11}
  ]
  rendererClasses [&NiagaraMeshRendererProperties, &NiagaraSpriteRendererProperties]
  renderers [
   {rendererClass=&NiagaraMeshRendererProperties, rendererIndex=0}
   {rendererClass=&NiagaraMeshRendererProperties, rendererIndex=1}
   {rendererClass=&NiagaraMeshRendererProperties, rendererIndex=2}
   {rendererClass=&NiagaraSpriteRendererProperties, rendererIndex=3}
  ]
 }
 Flames_PerTri {
  bEnabled=false
  simTarget=GPUComputeSim
  emitterUpdateScript [
   &EmitterState(
    Inactive Response=::NewEnumerator1
    Life Cycle Mode=::NewEnumerator1
    Loop Behavior=::NewEnumerator1
    Loop Duration Mode=::NewEnumerator1
    UseLoopDelay=false
   )
   &SpawnRate(SpawnRate=10, Use Spawn Probability=false)
   &SpawnBurst_Instantaneous(Spawn Count=4, Use Spawn Probability=false, UseLoopCountLimit=false)
  ]
  particleSpawnScript [
   &PartitionParticles(Number Of Partitions=25)
   &InitializeParticle(
    Color=<rgba 0.0018 0.0885 0.0857 1>
    Lifetime Max=12
    Lifetime Min=6
    Lifetime Mode=::NewEnumerator1
    Mesh Scale Mode=::NewEnumerator3
    Sprite Size Mode=::NewEnumerator3
    UsePositionOffset=false
    Write Material Random=true
   )
   &InitialMeshOrientation(
    Mesh Axis To Orient=<0 0 -1>
    Mesh Orientation Mode=::NewEnumerator2
    Orientation Coordinate Space=::World
    Orientation Vector=<0 0 1>
    Rotation=<0.25 0 0>
    Use Rotation Vector=false
   )
   &DynamicMaterialParameters(
    Index 0 Param 1=&RandomRangeFloat(Minimum=0, Maximum=1)
    Index 0 Param 2=2
    Module.Index0Param0Write=true
    Module.Index0Param1Write=true
    Module.Index0Param2Write=true
    Module.Index0Param3Write=true
   )
   &CameraOffset(Camera Offset Amount=35)
  ]
  particleUpdateScript [
   &ParticleState()
   /SetVariables(Particles.AlphaScale=1) {isSetParameters=true, module=SetVariables_#5}
   &ScaleMeshSize(Scale Factor=<1 1 2>)
   &SolveForcesAndVelocity(Clamp Velocity=false, Limit Acceleration=false)
   &SkeletalMeshLocation(
    MaskSpawnGroup=false
    Mesh Sampling Type=::NewEnumerator7
    OffsetAlongNormal=false
    Orientation Sampling=::NewEnumerator1
    OverrideRandomSeed=false
    Skeletal Mesh=$User.SK
   )
   &ScaleColor(Scale Alpha=$Particles.AlphaScale, Scale RGB=&VectorFromFloat(Value=3.5), ScaleA=true, ScaleRGB=true)
   /SetVariables(Particles.Color_02:LinearColor=<rgba 0.0018 0.0885 0.0857 1>) {isSetParameters=true, module=SetVariables_#6}
   /SetVariables(Particles.Color_03:LinearColor=<rgba 0.2203 0.9 0.7754 1>) {isSetParameters=true, module=SetVariables_#7}
   &InterpolateOverTime(Rate Of Change=3, Target Position Value:NiagaraPosition=$Particles.Position, Variable Type=::NewEnumerator6)
   /SetVariables(Particles.Interpolate Dynamic=&MakeFloat4FromVectorAndFloat(
     "Vector (RGB)"=$Particles.InterpolateOverTime.InterpolateOverTime_Position.Moving Average
     "Float (Alpha)"=&Add_Float(A=$User.Interpolate Intensity, B=-0.5)
    )) {isSetParameters=true, module=SetVariables_#8}
   &FN_Sprite_Attributes(
    Camera Offset:NiagaraFloat=&SelectFloatFromArray(
     Direct Array Index=$Particles.PartitionParticles.Partition
     Float Selection Array=DataInterface() {propertyValues {FloatData [2,2,5,5,5,2,2,5,0,5,2,1.5,1.5,1,1,1.5,1,1.5]}}
    )
    Color=<rgba 0.0473 0.4323 0.3617 1>
    Create Camera Offset Attribute=true
    Create Color Attribute=true
    Write to Module Namespace=true
   )
   /SetVariables(Particles.AirglowDyn:Vector4f=<2 1 0 0>) {isSetParameters=true, module=SetVariables_#9}
   /SetVariables(
    Particles.DecalDyn:Vector4f=&MakeVector4(
     X=0.8
     Y=0.5
     Z=&RandomRangeFloat(Minimum=0, Maximum=1)
     W=&RandomRangeFloat(Minimum=0, Maximum=0)
    )
    Particles.FireDecalColor:LinearColor=<rgba 0.1721 0.6323 0.7031 1>
    Particles.FireDecalSize:Vector2f=&Multiply_Vector2D_ByFloat(
     Float=&Multiply_Float(
      A=&SelectFloatFromArray(
       Direct Array Index=$Particles.PartitionParticles.Partition
       Float Selection Array=DataInterface() {propertyValues {FloatData [0.5,0.5,0.3,0.5,0.8,0.5,0.5,0.3,0.5,0.8,1,0.5,0.3,0.5,0.3,0.5,0.3,0.5,0,0,0,0,0,0,0]}}
      )
      B=1.2
     )
     Vector2D=$Particles.SpriteSize
    )
    Particles.GlowDecalColor:LinearColor=<rgba 0.0054 0.1413 0.195 1>
   ) {isSetParameters=true, module=SetVariables_#10}
   /SetVariables(Particles.VisibilityTag=&MakeCustomIntFromBool(
     Bool=$User.bisFootHidden
     True INT=&self:ScratchDynamicInput(
      Masked Int:NiagaraInt32=1
      Partition Mask:NiagaraInt32=&SelectIntFromArray(
       Direct Array Index=&ReturnExecIndex()
       Int Selection Array=DataInterface() {propertyValues {IntData [-1,-1,-1,-1,-1,-1,-1,-1,-1,-1,-1,-1,-1,-1,-1,-1,-1,-1,18,19,20,21,22,23,24,25]}}
      )
      Passthrough Int:NiagaraInt32=0
     )
     False INT=0
    )) {isSetParameters=true, module=SetVariables_#11}
   &DoOnce()
   &ScaleColor(
    Scale Alpha=&FloatFromCurve(
     FloatCurve=DataInterface() {propertyValues {Curve {keys [<at -0.25 0>,<at 0 1>,<at 0.25 0>]}}}
     CurveIndex=&DotProduct(
      Vector A=$Particles.SkeletalMeshLocation.SampledNormal
      Vector B=&ReturnCameraVector(Camera Vectors=::NewEnumerator4, Coordinate Space=::Simulation, Origin Position=$Particles.Position)
     )
    )
    ScaleA=true
    ScaleRGB=true
   ) {enabled=false, module=ScaleColor001}
   &ScaleMeshSize(Scale Factor=&VectorFromFloat(Value=&FloatFromCurve(
      FloatCurve=DataInterface() {propertyValues {Curve {keys [<at -1 0 RCIM_Constant>,<at -0.35 1 RCIM_Constant>,<at 0.35 0 RCIM_Constant>]}}}
      CurveIndex=&DotProduct(
       Vector A=$Particles.SkeletalMeshLocation.SampledNormal
       Vector B=&ReturnCameraVector(Camera Vectors=::NewEnumerator4, Coordinate Space=::Simulation, Origin Position=$Particles.Position)
      )
     ))) {enabled=false, module=ScaleMeshSize001}
   &KillParticles(Kill Particles=&SetBoolByFloatComparison(
     A=&FloatFromCurve(
      FloatCurve=DataInterface() {propertyValues {Curve {keys [<at -1 0 RCIM_Constant>,<at 0.0281 0>]}}}
      CurveIndex=&DotProduct(Vector A=$Particles.SkeletalMeshLocation.SampledNormal, Vector B=<0 0 1>)
     )
     B=0.5
    )) {enabled=false}
   &ScaleColor(
    Scale Alpha=&FloatFromCurve(FloatCurve=DataInterface() {propertyValues {Curve {keys [
         {arriveTangent=5.0583, interpMode=RCIM_Cubic, leaveTangent=5.0583, tangentMode=RCTM_User, time=0, value=0}
         {interpMode=RCIM_Cubic, tangentMode=RCTM_User, time=0.5, value=1}
         {arriveTangent=-3.9106, tangentMode=RCTM_User, time=1, value=0}
        ]}}})
    ScaleA=true
    ScaleRGB=true
   ) {enabled=false, module=ScaleColor002}
  ]
  rendererClasses [&NiagaraMeshRendererProperties]
  renderers [
   {rendererClass=&NiagaraMeshRendererProperties, rendererIndex=0}
   {rendererClass=&NiagaraMeshRendererProperties, rendererIndex=1}
   {rendererClass=&NiagaraMeshRendererProperties, rendererIndex=2}
  ]
 }
 Glow {
  emitterUpdateScript [
   &EmitterState(
    Inactive Response=::NewEnumerator1
    Life Cycle Mode=::NewEnumerator1
    Loop Behavior=::NewEnumerator1
    Loop Duration Mode=::NewEnumerator1
    UseLoopDelay=false
   )
   &SpawnBurst_Instantaneous(
    Spawn Count=&MakeCustomIntFromBool(
     Bool=$User.bIsFrontEndPreview
     True INT=1
     False INT=&MakeCustomIntFromBool(Bool=$User.bIsLocalPlayer, True INT=1, False INT=0)
    )
    Use Spawn Probability=false
    UseLoopCountLimit=false
   )
  ]
  particleSpawnScript [&InitializeParticle(
    Color=<rgba 0 1 0.7085 0.2>
    Mesh Scale=&Multiply_VectorByFloat(Vector=<1.8 0.8 0.8>, Float=1.25)
    Mesh Scale Mode=::NewEnumerator4
    Sprite Size=<118.0405 198.6727>
    Sprite Size Mode=::NewEnumerator4
    UsePositionOffset=false
    Write Material Random=true
   )]
  particleUpdateScript [
   &ParticleState(Kill Particles When Lifetime Has Elapsed=false, Let Infinitely Lived Particles Die When Emitter Deactivates=true)
   &SkeletalMeshLocation(Bone Sampling Mode=::NewEnumerator4, MaskSpawnGroup=false, Skeletal Mesh=$System.TargetSM)
   &CameraOffset(Camera Offset Amount=35)
   &ScaleColor(
    Scale Alpha=&FloatFromCurve(FloatCurve=DataInterface() {propertyValues {Curve {keys [<at 0 0>,<at 0.4 1>]}}}, CurveIndex=$Emitter.Age, Scale Curve=3)
    ScaleA=true
    ScaleRGB=true
   )
  ]
  rendererClasses [&NiagaraMeshRendererProperties]
  renderers [
   {rendererClass=&NiagaraMeshRendererProperties, rendererIndex=0}
   {rendererClass=&NiagaraMeshRendererProperties, rendererIndex=1}
   {rendererClass=&NiagaraMeshRendererProperties, rendererIndex=2}
  ]
 }
 Minimal {
  bEnabled=false
  emitterUpdateScript [
   &EmitterState()
   &SpawnBurst_Instantaneous(Use Spawn Probability=false, UseLoopCountLimit=false)
  ]
  particleSpawnScript [&InitializeParticle(Color Mode=::NewEnumerator0, Lifetime=5, UsePositionOffset=false, Write Material Random=true)]
  particleUpdateScript [&ParticleState()]
  rendererClasses [&NiagaraSpriteRendererProperties, &NiagaraComponentRendererProperties]
  renderers [
   {rendererClass=&NiagaraSpriteRendererProperties}
   {rendererClass=&NiagaraComponentRendererProperties, rendererIndex=1}
  ]
 }
}
summary {
 systemName=NS_GrillMoonPlayerAura
 emitters [
  {emitterName=Glow, rendererClasses [&NiagaraMeshRendererProperties]}
  {
   bEnabled=false
   emitterName=Minimal
   rendererClasses [&NiagaraSpriteRendererProperties, &NiagaraComponentRendererProperties]
  }
  {emitterName=Flames_PerBone, rendererClasses [&NiagaraMeshRendererProperties, &NiagaraSpriteRendererProperties]}
  {bEnabled=false, emitterName=Flames_PerTri, simTarget=GPUComputeSim, rendererClasses [&NiagaraMeshRendererProperties]}
  {
   bEnabled=false
   emitterName=FlamesArms_legs
   rendererClasses [&NiagaraMeshRendererProperties, &NiagaraSpriteRendererProperties]
  }
 ]
 userVariables [
  User.bisFootHidden {type=NiagaraBool}
  User.bIsFrontEnd {type=NiagaraBool}
  User.bIsFrontEndPreview {type=NiagaraBool}
  User.bIsLocalPlayer {type=NiagaraBool, defaultValue=true}
  User.bisLowLOD {type=NiagaraBool}
  User.GlowOn {type=NiagaraFloat, defaultValue=1, description=""}
  User.IdleGlow {type=NiagaraFloat, defaultValue=0, description=""}
  User.Interpolate Intensity {type=NiagaraFloat, defaultValue=1, description=""}
  User.SK {
   type=NiagaraDataInterfaceSkeletalMesh
   defaultValue {
    struct=NiagaraExt_VariableValue_DataInterface
    value {dataInterface=&self:NiagaraDataInterfaceSkeletalMesh_0}
   }
  }
  User.StartFrontEndLoop {type=NiagaraBool, defaultValue=true}
  User.Variant {type=NiagaraFloat, defaultValue=0, description=""}
 ]
}
